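# Sound cue for a MaskMaker voice-over effort line (take 020) in Section01 of the PistolCanary event audio.
# A single wave player feeds the matching wave asset straight to the output, with no randomization or mixing.
# It plays through the MaskMaker dialogue attenuation settings and the Fortnite events VO sound class.
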SoundCue /PistolCanary_Event_Audio/Assets/Section01/MaskMaker_VO/VO_MaskMaker_PistolCanary_Effort_020_Cue
root=WavePlayer_0
nodes [{
  id=WavePlayer_0
  properties {soundWaveAssetPtr=/PistolCanary_Event_Audio/Assets/Section01/MaskMaker_VO/wav/VO_MaskMaker_PistolCanary_Effort_020}
 }]
properties {
 attenuationSettings=/PistolCanary_Audio/Assets/Settings/ATT_PistolCanary_MaskMaker_Dialogue
 maxDistance=23000
 soundClassObject=/Game/Sounds/FortSoundClassesAndModes/Fort_Events_VO
 volumeMultiplier=1
}
