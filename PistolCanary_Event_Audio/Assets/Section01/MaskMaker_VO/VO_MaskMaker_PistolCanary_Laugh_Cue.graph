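# Sound cue that plays the MaskMaker laugh voice line for the PistolCanary event's Section01 audio.
# A single Dialogue Player node plays the laugh Dialogue Wave through a context whose speaker is the shared AudioLogVO voice, with no targets.
# The cue routes to the events VO sound class, uses the MaskMaker dialogue attenuation and slightly boosts volume.
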SoundCue /PistolCanary_Event_Audio/Assets/Section01/MaskMaker_VO/VO_MaskMaker_PistolCanary_Laugh_Cue
root=DialoguePlayer_5
nodes [{
  id=DialoguePlayer_5
  properties {dialogueWaveParameter {
    dialogueWave=/PistolCanary_Event_Audio/Assets/Section01/MaskMaker_VO/wav/VO_MaskMaker_PistolCanary_Laugh_Dialogue
    context {speaker=/Game/Sounds/Fort_Audio_Blueprints/Utilities/EditorUtilityWidgets/VO_Utilities/Assets/AudioLogVO_DV, targets []}
   }}
 }]
properties {
 attenuationSettings=/PistolCanary_Audio/Assets/Settings/ATT_PistolCanary_MaskMaker_Dialogue
 maxDistance=23000
 soundClassObject=/Game/Sounds/FortSoundClassesAndModes/Fort_Events_VO
 volumeMultiplier=1.3
}
