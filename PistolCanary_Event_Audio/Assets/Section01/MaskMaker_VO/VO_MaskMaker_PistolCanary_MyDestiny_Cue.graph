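# Sound cue for the Mask Maker "MyDestiny" voice-over line in Section 01 of the PistolCanary event.
# A single dialogue player node plays the line's dialogue wave, with the AudioLogVO dialogue voice as speaker and no targets.
# It uses the Mask Maker dialogue attenuation and the events VO sound class, and sends to a music-volume sidechain bus.
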SoundCue /PistolCanary_Event_Audio/Assets/Section01/MaskMaker_VO/VO_MaskMaker_PistolCanary_MyDestiny_Cue
root=DialoguePlayer_0
nodes [{
  id=DialoguePlayer_0
  properties {dialogueWaveParameter {
    dialogueWave=/PistolCanary_Event_Audio/Assets/Section01/MaskMaker_VO/wav/VO_MaskMaker_PistolCanary_MyDestiny_Dialogue
    context {speaker=/Game/Sounds/Fort_Audio_Blueprints/Utilities/EditorUtilityWidgets/VO_Utilities/Assets/AudioLogVO_DV, targets []}
   }}
 }]
properties {
 attenuationSettings=/PistolCanary_Audio/Assets/Settings/ATT_PistolCanary_MaskMaker_Dialogue
 maxDistance=23000
 soundClassObject=/Game/Sounds/FortSoundClassesAndModes/Fort_Events_VO
 volumeMultiplier=1
 busSends [{audioBus=&AB_Sidechain_FortMusic_Vol, minSendLevel=0, sendLevel=1, soundSourceBus=null}]
}
